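# Unlit translucent Niagara sprite material for the ball effect on the TimeInterval backpack cosmetic, drawing a white disc with a particle-colored band at its rim.
# A SphereMask on the UV distance from center forms the band, sized by a Thickness parameter, and blends white to particle color into MF_FX_EmissiveAdjustments.
# A second centered SphereMask, multiplied by particle alpha, cuts the circular shape procedurally without textures.
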Material /BRCosmetics/Accessories/FORT_Backpacks/Backpack_F_MED_TimeInterval/FX/M_Backpack_TimeInterval_Ball
graph {
 properties {bUsedWithNiagaraSprites=true, blendMode=BLEND_Translucent, pixelDepthOffsetMode=PDOM_Legacy, shadingModel=MSM_Unlit, usedShadingModels=Unlit}
 nodes [
  n1 = ParticleColor()
  n2 = Constant() {r=0.5}
  n3 = Distance(A=TextureCoordinate(), B=@n2)
  n4 = ScalarParameter() {defaultValue=0.075, parameterName=Thickness}
  n5 = Subtract(A=@n2, B=@n4)
  MaterialFunctionCall(Emissive=LinearInterpolate(A=Constant3Vector() {constant=<rgba 1 1 1 0>}, B=@n1.RGB, Alpha=SphereMask(A=@n3, B=@n5, Radius=@n4) {attenuationRadius=0.02})) {materialFunction=&MF_FX_EmissiveAdjustments}
  Multiply(A=@n1.A, B=SphereMask(A=TextureCoordinate(), B=Constant() {r=0.5}) {attenuationRadius=0.5})
 ]
}
