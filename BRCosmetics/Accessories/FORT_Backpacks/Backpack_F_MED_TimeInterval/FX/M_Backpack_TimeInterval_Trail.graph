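# Additive unlit Niagara ribbon and sprite material for the trail effect on the Time Interval backpack.
# Particle color feeds an emissive adjustment function, and particle alpha scales a procedural UV mask.
# The Glow static switch selects either a plain U falloff or one shaped by a powered sine across V.
# The glow branch adds an optional view-angle fade, off by default and bypassed at Low and Medium quality.
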Material /BRCosmetics/Accessories/FORT_Backpacks/Backpack_F_MED_TimeInterval/FX/M_Backpack_TimeInterval_Trail
graph {
 properties {
  bEnableResponsiveAA=true
  bOutputTranslucentVelocity=true
  bUsedWithNiagaraRibbons=true
  bUsedWithNiagaraSprites=true
  blendMode=BLEND_Additive
  pixelDepthOffsetMode=PDOM_Legacy
  shadingModel=MSM_Unlit
  usedShadingModels=Unlit
 }
 nodes [
  n1 = ParticleColor()
  MaterialFunctionCall(Emissive=@n1.RGB) {materialFunction=&MF_FX_EmissiveAdjustments}
  n2 = TextureCoordinate()
  n3 = ComponentMask(None=@n2) {g=true}
  n4 = OneMinus(None=ComponentMask(None=@n2) {r=true})
  n5 = Multiply(A=Power(Base=Sine(None=@n3) {period=2}), B=@n4)
  n6 = LinearInterpolate(
   A=Constant() {r=1}
   B=MaterialFunctionCall("FadeSensitivity (optional)"=ScalarParameter() {defaultValue=2, parameterName=AngleFadePower}) {materialFunction=&FadeBasedOnViewAngle}.Result
   Alpha=Sign(None=ScalarParameter() {parameterName=UseViewAngleFade})
  )
  n7 = Constant() {r=1}
  n8 = QualitySwitch(Default=@n6, Low=@n7, High=@n6, Medium=@n7, Epic=@n6)
  Multiply(A=@n1.A, B=StaticSwitchParameter(True=Multiply(A=@n5, B=@n8), False=@n4) {parameterName=Glow})
 ]
}
