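# Animation Blueprint for the TimeInterval backpack cosmetic, giving its pack bone a floating, springy motion.
# The AnimGraph offsets the bone, runs bone-relative Anim Dynamics with linear springs and a spherical limit, then adds a sine-driven hover translation.
# Each update computes the hover from game time and sets simulation alpha and velocity/acceleration scales through chained crouch, sprint and jump/fall branches.
# The final jump/fall branch always runs, overwriting the alpha and velocity scale set earlier.
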AnimBlueprint /BRCosmetics/Accessories/FORT_Backpacks/Backpack_F_MED_TimeInterval/Meshes/F_MED_TimeInterval_Pack_AnimBP
target_skeleton=&/F_MED_TimeInterval_Pack_Skeleton
graphs [
 AnimGraph {nodes [
   n1 = "Transform (Modify) Bone - Bone: pack"(ComponentPose=Local To Component(LocalPose=Input Pose()), Translation="0,-2.5,0", Rotation="0,0,0") {
    translationMode=BMM_Additive
    translationSpace=BCS_BoneSpace
    alphaBoolBlend {blendOption=Linear}
    boneToModify {boneName=pack}
    translation {y=-2.5}
   }
   n2 = Anim Dynamics - Bone: pack!(
    ComponentPose=@n1
    Alpha=Get Alpha_anim_dynamic() -> (Alpha_anim_dynamic:double)
    GravityOverride:Vector="0,0,20"
    ComponentLinearAccScale:Vector=Get Linear_Acc_Scale() -> (Linear_Acc_Scale:Vector)
    ComponentLinearVelScale:Vector=Get Linear_Vel_Scale() -> (Linear_Vel_Scale:Vector)
    ExternalForce="0,0,0"
   ) -> (Pose) {
    alpha=0
    angularBiasOverride=1
    angularDampingOverride=0.75
    angularSpringConstant=20
    bGravityOverrideInSimSpace=true
    bLinearSpring=true
    bOverrideLinearDamping=true
    bUseGravityOverride=true
    bUsePlanarLimit=false
    bUseSphericalLimits=true
    gravityScale=0
    linearDampingOverride=0.999
    linearSpringConstant=22
    simulationSpace=BoneRelative
    alphaBoolBlend {blendOption=Linear}
    alphaScaleBiasClamp {bInterpResult=true}
    boundBone {boneName=pack}
    componentAppliedLinearAccClamp=<65 55 50>
    componentLinearAccScale=<0.7 0.7 1.5>
    componentLinearVelScale=<0.7 0.7 1.5>
    gravityOverride {z=20}
    physicsBodyDefinitions [{
      sphereCollisionRadius=0
      boundBone {boneName=pack}
      boxExtents=<20 8 20>
      constraintSetup {
       linearXLimitType=Free
       linearZLimitType=Free
       angularLimitsMax=<0 0 0>
       angularLimitsMin=<0 0 0>
       angularTarget=<0 0 0>
       linearAxesMax=<2.5 0.5 0>
       linearAxesMin=<-2.5 -10 0>
      }
      localJointOffset=<0 0 0>
     }]
    relativeSpaceBone {boneName=root}
    sphericalLimits [{limitRadius=7, limitType=Inner, drivingBone {boneName=pack}, sphereLocalOffset=<0 0 4.4>}]
   }
   n3 = "Transform (Modify) Bone - Bone: pack"(ComponentPose=@n2, Translation=Get Backpack_floating() -> (Backpack_floating:Vector), Rotation="0,0,0") {translationMode=BMM_Additive, translationSpace=BCS_BoneSpace, alphaBoolBlend {blendOption=Linear}, boneToModify {boneName=pack}}
   Output Pose(Result=Component To Local(ComponentPose=@n3)) {layerGroup=None}
  ]}
 EventGraph {nodes [
   Event BlueprintUpdateAnimation()
   > n1 = Sequence()
   n2 = float * float(A=GetGameTimeInSeconds(), B=2) {is=title K2Node_PromotableOperator}
   n3 = "Sin (Radians)"(A=float * float(A=1.21, B=@n2) {is=title K2Node_CommutativeAssociativeBinaryOperator})
   n4 = "Sin (Radians)"(A=float * float(A=0.6, B=@n2))
   n5 = "Sin (Radians)"(A=float * float(A=0.8, B=@n2))
   > Set Backpack_floating(Backpack_floating:Vector=MakeVector(X=float * float(A=@n4, B=0.5), Y=float * float(A=@n3, B=0.15), Z=float * float(A=@n5, B=1))) -> (Output_Get:Vector)
   n6 = Branch(execute=@n1.then_1, Condition=Get bIsCrouching())
   > Set Alpha_anim_dynamic(Alpha_anim_dynamic:double=0.3) -> (Output_Get:double)
   > Set Linear_Acc_Scale(Linear_Acc_Scale:Vector="0,0,0.1") -> (Output_Get:Vector)
   > n7 = Set Linear_Vel_Scale(Linear_Vel_Scale:Vector="1.2,1.2,0.6") -> (Output_Get:Vector)
   n8 = Get bIsCrouching()
   n9 = Branch(execute=[@n13.then, @n7.then], Condition=AND Boolean(A=Get bIsSprinting(), B=NOT Boolean(A=@n8)))
   > Set Alpha_anim_dynamic(Alpha_anim_dynamic=0.1)
   > Set Linear_Acc_Scale(Linear_Acc_Scale="0,0,0.1")
   > n10 = Set Linear_Vel_Scale(Linear_Vel_Scale="1,1,1.2")
   n11 = Branch(execute=[@n12.then, @n10.then], Condition=OR Boolean(A=Get bIsJumping(), B=Get bIsFalling()))
   > Set Alpha_anim_dynamic(Alpha_anim_dynamic=0.1)
   > Set Linear_Acc_Scale(Linear_Acc_Scale="0, 0, 0")
   > Set Linear_Vel_Scale(Linear_Vel_Scale="0.2,0.2,0.8")
   Set Alpha_anim_dynamic(execute=@n11.else, Alpha_anim_dynamic=0.7)
   > Set Linear_Acc_Scale(Linear_Acc_Scale="1.1,1.1,1.3")
   > Set Linear_Vel_Scale(Linear_Vel_Scale="1.1,1.1,1.3")
   Set Alpha_anim_dynamic(execute=@n9.else, Alpha_anim_dynamic=0.7)
   > Set Linear_Vel_Scale(Linear_Vel_Scale="1.1,1.1,1.3")
   > n12 = Set Linear_Vel_Scale(Linear_Vel_Scale="1.1,1.1,1.3")
   Set Alpha_anim_dynamic(execute=@n6.else, Alpha_anim_dynamic=0.7)
   > Set Linear_Vel_Scale(Linear_Vel_Scale="1.1,1.1,1.3")
   > n13 = Set Linear_Vel_Scale(Linear_Vel_Scale="1.1,1.1,1.3")
   TryGetPawnOwner()
  ]}
]
variables [
 Alpha_anim_dynamic {type=number}
 Backpack_floating {type=Vector}
 Linear_Acc_Scale {type=Vector}
 Linear_Vel_Scale {type=Vector}
]
